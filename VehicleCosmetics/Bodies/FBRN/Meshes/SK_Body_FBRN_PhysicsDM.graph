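# Physics asset for the FBRN vehicle body skeletal mesh in VehicleCosmetics, giving it a simple collision volume.
# It holds a single body on the root bone with one large box shape, raised above the origin to enclose the vehicle body.
# It has no constraints, so the body has no articulated physics parts.
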PhysicsAsset /VehicleCosmetics/Bodies/FBRN/Meshes/SK_Body_FBRN_PhysicsDM
bodies {root [{shapeType=Box, extentX=444, extentY=216, extentZ=132, shapeName=root_box, center=<10.2585 0 95>}]}
constraints []
